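# Looping ambient cue for the Leaky object in the Dopey event, routed to the map-wide events sound class.
# A BeamCrossfade parameter crossfades from a snow ambience loop to Leaky's full-power loop, while a continuous modulator uses it to raise pitch slightly.
# Each loop sits under close attenuation, overridden for the snow loop, in a mixer whose distant-attenuation branch has no sound attached.
# Concurrency is capped, stopping the oldest instance.
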SoundCue /Figment_S08/Athena/Sounds/Events/Dopey/Leaky/SC_Leaky_Amb_Cue
root=ModulatorContinuous_1
nodes [
 {
  id=Attenuation_0
  children [WavePlayer_0]
  properties {
   attenuationSettings=/Game/Athena/Sounds/Events/Dopey/Sleepy/Sleepy_MLTI_Close_Att
   bOverrideAttenuation=true
   attenuationOverrides {
    bAttenuateWithLPF=true
    bEnableFocusInterpolation=true
    bEnableListenerFocus=true
    falloffDistance=2000
    lPFFrequencyAtMax=2000
    lPFRadiusMax=2000
    lPFRadiusMin=1000
    nonFocusVolumeAttenuation=0.5
    reverbDistanceMax=15000
    reverbWetLevelMax=20
    reverbWetLevelMin=0
    stereoSpread=512
    attenuationShapeExtents {x=0}
   }
  }
 }
 {id=Attenuation_1, children [null], properties {attenuationSettings=/Game/Athena/Sounds/Events/Dopey/Sleepy/Sleepy_MLTI_Distant_Att}}
 {id=Attenuation_3, children [null], properties {attenuationSettings=&Sleepy_MLTI_Distant_Att}}
 {id=Attenuation_4, children [WavePlayer_1], properties {attenuationSettings=&Sleepy_MLTI_Close_Att}}
 {id=Mixer_0, children [Attenuation_0, Attenuation_1], properties {inputVolume [1,1]}}
 {id=Mixer_1, children [Attenuation_4, Attenuation_3], properties {inputVolume [1,1]}}
 {
  id=ModulatorContinuous_1
  children [ParamCrossFade_0]
  properties {pitchModulationParams {default=0, maxOutput=1.15, minOutput=1, parameterName=BeamCrossfade}}
 }
 {
  id=ParamCrossFade_0
  children [Mixer_0,Mixer_1]
  properties {
   paramName=BeamCrossfade
   crossFadeInput [
    {fadeInDistanceEnd=0, fadeInDistanceStart=0, fadeOutDistanceEnd=1, fadeOutDistanceStart=0, volume=1}
    {fadeInDistanceEnd=1, fadeInDistanceStart=0, fadeOutDistanceEnd=100, fadeOutDistanceStart=100, volume=1}
   ]
  }
 }
 {id=WavePlayer_0, properties {bLooping=true, soundWaveAssetPtr=/Figment_S08/Athena/Sounds/Events/Dopey/Snow/Snow_Ambient_02}}
 {id=WavePlayer_1, properties {bLooping=true, soundWaveAssetPtr=&/Leaky_Full_Power_Start_Loop}}
]
properties {
 bOverrideConcurrency=true
 maxDistance=35000
 soundClassObject=/Figment_S08/Athena/Sounds/Events/SoundClasses/Fort_Events_Mapwide
 volumeMultiplier=1
 attenuationOverrides {falloffDistance=3000, attenuationShapeExtents {x=0}}
 concurrencyOverrides {maxCount=4, resolutionRule=StopOldest}
}
