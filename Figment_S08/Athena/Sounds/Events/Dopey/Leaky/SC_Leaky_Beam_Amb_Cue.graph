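# Looping ambient beam sound cue for the Leaky event under Dopey, reusing the CrackLightning beam loop wave.
# A looping wave player feeds a smoothed, high-intensity Doppler node at the root, routed to the map-wide events sound class.
# A concurrency override caps simultaneous instances, stopping the oldest.
# Capsule attenuation values are stored, but the attenuation override is not enabled and no attenuation asset is assigned.
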SoundCue /Figment_S08/Athena/Sounds/Events/Dopey/Leaky/SC_Leaky_Beam_Amb_Cue
root=Doppler_0
nodes [
 {id=Doppler_0, children [WavePlayer_0], properties {bUseSmoothing=true, dopplerIntensity=10}}
 {id=WavePlayer_0, properties {bLooping=true, soundWaveAssetPtr=/Game/Athena/Sounds/Events/CrackLightning/Crack_Lightning_BeamLoop_01}}
]
properties {
 bOverrideConcurrency=true
 maxDistance=2097152
 soundClassObject=/Figment_S08/Athena/Sounds/Events/SoundClasses/Fort_Events_Mapwide
 volumeMultiplier=1
 attenuationOverrides {attenuationShape=Capsule, falloffDistance=512, attenuationShapeExtents=<2000 512>}
 concurrencyOverrides {maxCount=4, resolutionRule=StopOldest}
}
