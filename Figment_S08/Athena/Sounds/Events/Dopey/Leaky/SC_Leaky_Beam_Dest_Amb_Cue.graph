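# Looping ambience cue for the Leaky event's beam destination, routed to the map-wide events sound class.
# A BeamCount switch feeds one looping full-power wave through three modulators, raising pitch in steps as the count goes from one to three.
# The unset and zero inputs are empty, so those states are silent.
# Overrides give it long-range capsule attenuation with low-pass and listener focus, plus stop-oldest concurrency.
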SoundCue /Figment_S08/Athena/Sounds/Events/Dopey/Leaky/SC_Leaky_Beam_Dest_Amb_Cue
root=Switch_0
nodes [
 {id=Modulator_0, children [WavePlayer_1], properties {pitchMax=1, pitchMin=1, volumeMax=1, volumeMin=1}}
 {id=Modulator_1, children [WavePlayer_1], properties {pitchMax=1.1, pitchMin=1.1, volumeMax=1, volumeMin=1}}
 {id=Modulator_2, children [WavePlayer_1], properties {pitchMax=1.2, pitchMin=1.2, volumeMax=1, volumeMin=1}}
 {id=Switch_0, children [null,null,Modulator_0,Modulator_1,Modulator_2], properties {intParameterName=BeamCount}}
 {id=WavePlayer_1, properties {bLooping=true, soundWaveAssetPtr=&/Leaky_Full_Power_Start_Loop}}
]
properties {
 bOverrideAttenuation=true
 bOverrideConcurrency=true
 maxDistance=32000
 soundClassObject=/Figment_S08/Athena/Sounds/Events/SoundClasses/Fort_Events_Mapwide
 volumeMultiplier=1
 attenuationOverrides {
  attenuationShape=Capsule
  bAttenuateWithLPF=true
  bEnableFocusInterpolation=true
  bEnableListenerFocus=true
  falloffDistance=30000
  focusAzimuth=2
  focusDistanceScale=0.2
  lPFFrequencyAtMax=5000
  lPFRadiusMax=10000
  lPFRadiusMin=5000
  nonFocusAzimuth=10
  nonFocusVolumeAttenuation=0
  stereoSpread=2048
  attenuationShapeExtents {x=2000}
 }
 concurrencyOverrides {maxCount=4, resolutionRule=StopOldest}
}
