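# Sound cue for the Leaky event's beam destination hit, playing one of three Snow pre-complete stage sounds.
# A fixed one-second delay feeds a switch on the BeamCount integer parameter, whose first two inputs are empty and whose remaining inputs hold the progressive stage waves.
# It uses shared Snow close/distant attenuation, the mapwide events sound class and a concurrency override that stops the oldest instance.
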SoundCue /Figment_S08/Athena/Sounds/Events/Dopey/Leaky/SC_Leaky_Beam_Dest_Hit_Cue
root=Delay_0
nodes [
 {id=Delay_0, children [Switch_0], properties {delayMax=1, delayMin=1}}
 {id=Switch_0, children [null, null, WavePlayer_0, WavePlayer_7, WavePlayer_10], properties {intParameterName=BeamCount}}
 {id=WavePlayer_0, properties {soundWaveAssetPtr=/Figment_S08/Athena/Sounds/Events/Dopey/Snow/Snow_Pre_Complete_Stage_01}}
 {id=WavePlayer_10, properties {soundWaveAssetPtr=/Figment_S08/Athena/Sounds/Events/Dopey/Snow/Snow_Pre_Complete_Stage_03}}
 {id=WavePlayer_7, properties {soundWaveAssetPtr=/Figment_S08/Athena/Sounds/Events/Dopey/Snow/Snow_Pre_Complete_Stage_02}}
]
properties {
 attenuationSettings=/Figment_S08/Athena/Sounds/Events/Dopey/Snow/Snow_MLTI_CloseDistant_Att
 bOverrideConcurrency=true
 maxDistance=2097152
 soundClassObject=/Figment_S08/Athena/Sounds/Events/SoundClasses/Fort_Events_Mapwide
 attenuationOverrides {falloffDistance=3000, attenuationShapeExtents {x=0}}
 concurrencyOverrides {maxCount=4, resolutionRule=StopOldest}
}
