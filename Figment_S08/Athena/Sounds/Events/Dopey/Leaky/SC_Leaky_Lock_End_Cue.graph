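# Sound cue for the end of a Leaky lock event in Figment Season 8, playing a power-down sound in the map-wide events sound class.
# A Mixer combines a close branch, where the wave runs through a Modulator and borrowed Sleepy close attenuation, with a distant attenuation branch that has no input, so only the close branch produces sound.
# A concurrency override caps simultaneous instances at four, stopping the oldest.
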SoundCue /Figment_S08/Athena/Sounds/Events/Dopey/Leaky/SC_Leaky_Lock_End_Cue
root=Mixer_0
nodes [
 {id=Attenuation_0, children [Modulator_0], properties {attenuationSettings=/Game/Athena/Sounds/Events/Dopey/Sleepy/Sleepy_MLTI_Close_Att}}
 {id=Attenuation_1, children [null], properties {attenuationSettings=/Game/Athena/Sounds/Events/Dopey/Sleepy/Sleepy_MLTI_Distant_Att}}
 {id=Mixer_0, children [Attenuation_0, Attenuation_1], properties {inputVolume [1,1]}}
 {id=Modulator_0, children [WavePlayer_1]}
 {id=WavePlayer_1, properties {soundWaveAssetPtr=&/Leaky_Power_Down_01}}
]
properties {
 bOverrideConcurrency=true
 maxDistance=35000
 soundClassObject=/Figment_S08/Athena/Sounds/Events/SoundClasses/Fort_Events_Mapwide
 volumeMultiplier=1
 attenuationOverrides {falloffDistance=3000, attenuationShapeExtents {x=0}}
 concurrencyOverrides {maxCount=4, resolutionRule=StopOldest}
}
